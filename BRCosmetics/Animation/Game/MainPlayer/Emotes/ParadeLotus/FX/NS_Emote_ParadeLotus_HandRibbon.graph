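# Niagara system for the ParadeLotus emote that draws short-lived ribbon trails from the player's RightHand and LeftHand skeletal sockets.
# A single SwordTrail emitter partitions particles between the sockets and orients ribbon facing from the sampled socket rotation.
# User.RibbonCurve drives everything: its magnitude scales spawn rate, while comparisons against zero choose the position offset and switch ribbon colors on or off, including per-partition color.
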NiagaraSystem /BRCosmetics/Animation/Game/MainPlayer/Emotes/ParadeLotus/FX/NS_Emote_ParadeLotus_HandRibbon
dependencies {
 usedDataInterfaces [&NiagaraDataInterfaceSkeletalMesh, &NiagaraDataInterfaceArrayInt32]
 usedDynamicInputs []
 usedModules [&SystemState]
}
emitters {SwordTrail {
  emitterUpdateScript [
   &EmitterState(Life Cycle Mode=::NewEnumerator1, Loop Behavior=::NewEnumerator1, Loop Duration Mode=::NewEnumerator1, UseLoopDelay=false)
   &SpawnRate(
    SpawnRate=&Abs_Float(Float:NiagaraFloat=&Multiply_Float(A=$User.RibbonCurve, B=1042.2352))
    Use Spawn Probability=false
   )
  ]
  particleSpawnScript [
   &PartitionParticles()
   &InitializeParticle(
    Lifetime=0.1
    Mass Max=1.5
    Mass Mode=::NewEnumerator3
    Ribbon Width=14.7
    Ribbon Width Mode=::NewEnumerator1
    Sprite Rotation Mode=::NewEnumerator2
    Sprite Size Mode=::NewEnumerator1
    Uniform Sprite Size Max=8
    Uniform Sprite Size Min=4
    UsePositionOffset=false
    Write Material Random=true
   )
   &SkeletalMeshLocation(
    Bone / Socket Index=&SelectIntFromArray(Direct Array Index=$Particles.PartitionParticles.Partition, Int Selection Array=DataInterface() {propertyValues {IntData [0,1]}})
    MaskSpawnGroup=false
    Mesh Sampling Type=::NewEnumerator5
    Orientation Sampling=::NewEnumerator1
    Sampled Position Offset=&MakeCustomVectorFromBool(Bool=&SetBoolByFloatComparison(A=$User.RibbonCurve, B=0), True Vector=<0 0 0>, False Vector=<0 10 0>)
    Skeletal Mesh=DataInterface() {propertyValues {
      DefaultMesh=null
      ExcludeBoneName=root
      OptionalComponentClass=null
      PreviewMesh=/Game/Characters/Player/Male/Male_Avg_Base/Fortnite_M_Avg_Player
      SkinningMode=SkinOnTheFly
      SoftSourceActor=null
      SourceMode=Default
      UvSetIndex=0
      WholeMeshLOD=-1
      bExcludeBone=false
      bReadDeformedGeometry=true
      bRequireCurrentFrameData=true
      ComponentTags []
      FilteredBones []
      FilteredSockets [RightHand,LeftHand]
      MeshUserParameter {parameter {name=None, typeDefHandle {registeredTypeIndex=1}, varData []}}
      SamplingRegions []
     }}
    Socket Sampling Mode=::NewEnumerator5
   )
  ]
  particleUpdateScript [
   &ParticleState()
   &SolveForcesAndVelocity(Clamp Velocity=false, Limit Acceleration=false)
   /SetVariables(Particles.RibbonFace:Vector3f=&RotateVectorByQuaternion(Vector To Rotate=<0 0 1>, Quaternion=$Particles.SkeletalMeshLocation.SampledRotation)) {isSetParameters=true}
   /SetVariables(
    Particles.Ribbon1Color:LinearColor=&Multiply_LinearColorByFloat(
     Linear Color=<rgba 1 1 1 1>
     Float=&MakeCustomFloatFromBool(Bool=&SetBoolByFloatComparison(A=$User.RibbonCurve, B=0), True Float=1, False Float=0)
    )
    Particles.Ribbon2Color:LinearColor=&Multiply_LinearColorByFloat(
     Linear Color=<rgba 1 1 1 1>
     Float=&MakeCustomFloatFromBool(
      Bool=&SetBoolByFloatComparison(A=$User.RibbonCurve, B=0, Comparison Type=::NewEnumerator6)
      True Float=1
      False Float=0
     )
    )
    Particles.Ribbon2Wiedth=60
    Particles.RibbonColorPart:LinearColor=&MergeHelper_ColorSwitch(
     0=&Multiply_LinearColorByFloat(
      Linear Color=<rgba 1 1 1 1>
      Float=&MakeCustomFloatFromBool(Bool=&SetBoolByFloatComparison(A=$User.RibbonCurve, B=0), True Float=0.3, False Float=0)
     )
     1=&Multiply_LinearColorByFloat(
      Linear Color=<rgba 1 1 1 1>
      Float=&MakeCustomFloatFromBool(
       Bool=&SetBoolByFloatComparison(A=$User.RibbonCurve, B=0, Comparison Type=::NewEnumerator6)
       True Float=0.3
       False Float=0
      )
     )
     2=?
     3=?
     Index=$Particles.PartitionParticles.Partition
     Num=::NewEnumerator0
    )
   ) {isSetParameters=true}
  ]
  rendererClasses [&NiagaraRibbonRendererProperties]
  renderers [{rendererClass=&NiagaraRibbonRendererProperties}]
 }}
summary {
 systemName=NS_Emote_ParadeLotus_HandRibbon
 emitters [{emitterName=SwordTrail, rendererClasses [&NiagaraRibbonRendererProperties]}]
 userVariables [User.RibbonCurve {type=NiagaraFloat, defaultValue=1, description=""}]
}
