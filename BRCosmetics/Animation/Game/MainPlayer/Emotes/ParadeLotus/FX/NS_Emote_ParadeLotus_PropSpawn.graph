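# Niagara system for the prop-spawn effect of the ParadeLotus player emote, built from airglow, sparkle, toon smoke and toon impact smoke emitters.
# System scripts set variables alongside SystemState, using time-of-day color and float-array selection dynamic inputs.
# A separate ToonSmoke_LowMedium emitter provides a smoke variant, apparently for lower quality levels.
# It exposes a NormalizedNotifyProgress user float, presumably fed by an animation notify.
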NiagaraSystem /BRCosmetics/Animation/Game/MainPlayer/Emotes/ParadeLotus/FX/NS_Emote_ParadeLotus_PropSpawn
dependencies {
 usedDataInterfaces [&NiagaraDataInterfaceArrayFloat]
 usedDynamicInputs [&FNTimeOfDayColor, &SelectFloatFromArray]
 usedModules [
  &self:SystemSpawnScript.SystemScriptSource.SystemScriptGraph.NiagaraNodeAssignment_1.SetVariables_#1
  &SystemState
  &self:SystemSpawnScript.SystemScriptSource.SystemScriptGraph.NiagaraNodeAssignment_0.SetVariables_#2
 ]
}
emitters {
 AirglowLWE {rendererClasses [], renderers []}
 SparklesLWE {rendererClasses [], renderers []}
 ToonImpactSmoke {rendererClasses [], renderers []}
 ToonSmoke {rendererClasses [], renderers []}
 ToonSmoke_LowMedium {rendererClasses [], renderers []}
}
summary {
 systemName=NS_Emote_ParadeLotus_PropSpawn
 emitters [
  {emitterName=AirglowLWE, rendererClasses []}
  {emitterName=SparklesLWE, rendererClasses []}
  {emitterName=ToonImpactSmoke, rendererClasses []}
  {emitterName=ToonSmoke, rendererClasses []}
  {emitterName=ToonSmoke_LowMedium, rendererClasses []}
 ]
 userVariables [User.NormalizedNotifyProgress {type=NiagaraFloat, defaultValue=0, description=""}]
}
