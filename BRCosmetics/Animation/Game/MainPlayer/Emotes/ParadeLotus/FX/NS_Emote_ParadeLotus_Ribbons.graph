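# Niagara system for the ParadeLotus emote, layering sword trails, punch bursts, sparkles and arrow meshes over the gadget skeletal mesh.
# SwordTrail samples a sword socket, with User.RibbonCurve driving spawn rate and width and its sign gating two time-of-day-tinted ribbons.
# Punchs spawns four sprite groups whenever User.BurstIndex changes, choosing among three fixed offsets.
# Sparkles follow sword sockets, while Arrows track arrowtip bones and scale meshes by distance from the emitter.
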NiagaraSystem /BRCosmetics/Animation/Game/MainPlayer/Emotes/ParadeLotus/FX/NS_Emote_ParadeLotus_Ribbons
dependencies {
 usedDataInterfaces [
  &NiagaraDataInterfaceSkeletalMesh
  &NiagaraDataInterfaceCurve
  &NiagaraDataInterfaceArrayFloat
  &NiagaraDataInterfaceArrayInt32
 ]
 usedDynamicInputs []
 usedModules [&SystemState]
}
emitters {
 ArrowSmearLines {rendererClasses [], renderers []}
 Arrows {
  emitterUpdateScript [
   &EmitterState(
    Inactive Response=::NewEnumerator1
    Life Cycle Mode=::NewEnumerator1
    Loop Behavior=::NewEnumerator1
    Loop Delay=6.5
    Loop Duration=3
    UseLoopDelay=true
   )
   &SpawnRate(SpawnGroup=2, SpawnRate=100, Use Spawn Probability=false)
   &SpawnBurst_Instantaneous(Spawn Count=40, Use Spawn Probability=false, UseLoopCountLimit=false)
   &SpawnBurst_Instantaneous(Spawn Count=20, Spawn Group=1, Use Spawn Probability=false, UseLoopCountLimit=false) {module=SpawnBurst_Instantaneous002}
  ]
  particleSpawnScript [
   /SetVariables(Particles.VisibilityTag=&SpawnGroup()) {isSetParameters=true}
   &PartitionParticles(Number Of Partitions=&MergeHelper_IntSwitch(0=0, 1=2, 2=?, 3=?, 4=?, 5=?, Index=&SpawnGroup(), Num=::NewEnumerator0))
   &InitializeParticle(
    Color=<rgba 0 0 0 1>
    Lifetime=&MergeHelper_FloatSwitch(
     0=5
     1=5
     2=&RandomRangeFloat(Minimum=0.15, Maximum=0.3)
     3=?
     Index=&SpawnGroup()
     Num=::NewEnumerator1
    )
    Mass Max=1.5
    Mass Mode=::NewEnumerator3
    Mesh Scale Mode=::NewEnumerator3
    Mesh Uniform Scale=1.2
    Sprite Size Max=<1.3278 10.5736>
    Sprite Size Min=<0.576 47.6671>
    Sprite Size Mode=::NewEnumerator2
    UsePositionOffset=false
    Write Material Random=true
   )
   &InitialMeshOrientation(
    Mesh Axis To Orient=<0 0 0>
    Mesh Orientation Mode=::NewEnumerator2
    Orientation Coordinate Space=::Local
    Rotation Coordinate Space=::NewEnumerator1
    Use Rotation Vector=true
   )
  ]
  particleUpdateScript [
   &ParticleState()
   &SkeletalMeshLocation(
    Bone / Socket Index=&MergeHelper_IntSwitch(
     0=&RandomRangeInt(Minimum=0, Maximum=15)
     1=&MergeHelper_IntSwitch(0=16, 1=17, 2=?, 3=?, 4=?, 5=?, Index=$Particles.PartitionParticles.Partition, Num=::NewEnumerator0)
     2=?
     3=?
     4=?
     5=?
     Index=&SpawnGroup()
     Num=::NewEnumerator0
    )
    Bone Sampling Mode=::NewEnumerator4
    MaskSpawnGroup=false
    Orientation Sampling=::NewEnumerator1
    Skeletal Mesh=DataInterface() {propertyValues {
      DefaultMesh=null
      ExcludeBoneName=root
      OptionalComponentClass=null
      PreviewMesh=/BRCosmetics/Animation/Props/Gadgets_ParadeLotus/Meshes/ParadeLotus_Gadget
      SkinningMode=SkinOnTheFly
      SoftSourceActor=null
      SourceMode=Default
      UvSetIndex=0
      WholeMeshLOD=1
      bExcludeBone=false
      bReadDeformedGeometry=true
      bRequireCurrentFrameData=true
      ComponentTags []
      FilteredBones [
       arrow12_Arrowtip_LOD0_JNT
       arrow13_Arrowtip_LOD0_JNT
       arrow14_Arrowtip_LOD0_JNT
       arrow15_Arrowtip_LOD0_JNT
       arrow16_Arrowtip_LOD0_JNT
       arrow2_Arrowtip_LOD0_JNT
       arrow18_Arrowtip_LOD0_JNT
       arrow10_Arrowtip_LOD0_JNT
       arrow11_Arrowtip_LOD0_JNT
       arrow9_Arrowtip_LOD0_JNT
       arrow8_Arrowtip_LOD0_JNT
       arrow7_Arrowtip_LOD0_JNT
       arrow6_Arrowtip_LOD0_JNT
       arrow5_Arrowtip_LOD0_JNT
       arrow4_Arrowtip_LOD0_JNT
       arrow3_Arrowtip_LOD0_JNT
       arrow17_Arrowtip_LOD0_JNT
       arrow1_Arrowtip_LOD0_JNT
      ]
      FilteredSockets []
      MeshUserParameter {parameter {name=None, typeDefHandle {registeredTypeIndex=1}, varData []}}
      SamplingRegions []
     }}
    Use Sampled Mesh Attribute as Ribbon ID=false
    Write To Particle Position=false
   )
   &ScaleMeshSize(Scale Factor=&VectorFromFloat(Value=&FloatFromCurve(
      FloatCurve=DataInterface() {propertyValues {Curve {
         defaultValue=1
         keys [{interpMode=RCIM_Cubic, tangentMode=RCTM_User, time=0.4081, value=1}, {interpMode=RCIM_Cubic, tangentMode=RCTM_User, time=1, value=0}]
        }}}
      CurveIndex=&NormalizeDistanceRange(
       Start Position:NiagaraPosition=$Engine.Emitter.SimulationPosition
       End Position:NiagaraPosition=$Particles.Position
       Distance:NiagaraFloat=&FN_FrontEndFloatModifier(Gameplay=410, Front End=300)
       Invert Normalized Range:NiagaraBool=false
      )
     )))
   &ShapeLocation(
    Additional Quaternion Rotation=true
    Additional Yaw / Pitch / Roll=&MergeHelper_VectorSwitch(
     0=<0 0 0>
     1=&MergeHelper_VectorSwitch(0=<0 0 -35>, 1=<0 0 44.7532>, 2=?, 3=?, Index=$Particles.PartitionParticles.Partition, Num=::NewEnumerator0)
     2=?
     3=?
     Index=&SpawnGroup()
     Num=::NewEnumerator0
    )
    MaskSpawnGroup=false
    Non Uniform Scale=&MergeHelper_VectorSwitch(0=<1 0.57 2.1902>, 1=<1 3.5135 0>, 2=?, 3=?, Index=&SpawnGroup(), Num=::NewEnumerator0)
    Offset:NiagaraPosition=$Particles.SkeletalMeshLocation.SampledPosition
    Offset Mode=::NewEnumerator0
    Override Seed=false
    Rotation Angle Type:ENiagara_AngleInput=::NewEnumerator0
    Rotation Mode=::NewEnumerator3
    Rotation Quaternion=$Particles.SkeletalMeshLocation.SampledRotation
    Sphere Radius=2.1975
    UseOwnerScale=true
   )
   &UpdateMeshOrientation(
    Facing Direction=&RotateVectorByQuaternion(Vector To Rotate=<1 0 0>, Quaternion=$Particles.SkeletalMeshLocation.SampledRotation)
    Facing Interpolation=false
    Orientation Method=::NewEnumerator1
    Override Quat=false
    TransformFacing=false
    TransformUp=false
    Up Direction:Vector3f=<0 0 0>
   )
   &SolveForcesAndVelocity(Clamp Velocity=false, Limit Acceleration=false)
  ]
  rendererClasses [&NiagaraMeshRendererProperties]
  renderers [
   {rendererClass=&NiagaraMeshRendererProperties, rendererIndex=0}
   {rendererClass=&NiagaraMeshRendererProperties, rendererIndex=1}
  ]
 }
 BurstLines {rendererClasses [], renderers []}
 Punchs {
  emitterUpdateScript [
   &EmitterState(Life Cycle Mode=::NewEnumerator1, Loop Behavior=::NewEnumerator1, Loop Duration=4.5319, UseLoopDelay=false)
   /Game/Effects/Niagara/Modules/Logic/UpdateValueStateChange(InputIntParam:NiagaraInt32=&MakeIntFromFloat(Float=$User.BurstIndex, Conversion Mode=::NewEnumerator1))
   &self:TriggerSpawn(
    Spawn Count:NiagaraInt32=&MergeHelper_IntSwitch(
     0=1
     1=1
     2=1
     3=?
     4=?
     5=?
     Index=&MakeIntFromFloat(Float=$User.BurstIndex, Conversion Mode=::NewEnumerator1)
     Num=::NewEnumerator1
    )
    Spawn Group:NiagaraInt32=&MakeIntFromFloat(Float=0, Conversion Mode=::NewEnumerator1)
    Trigger:NiagaraBool=&Boolean_AndOperation(Bool1=$Emitter.UpdateValueStateChange.Changed, Bool2=&SetBoolByFloatComparison(A=$System.Age, B=0))
   ) {module=TriggerSpawn}
   &self:TriggerSpawn(
    Spawn Count:NiagaraInt32=&MergeHelper_IntSwitch(
     0=1
     1=1
     2=1
     3=?
     4=?
     5=?
     Index=&MakeIntFromFloat(Float=$User.BurstIndex, Conversion Mode=::NewEnumerator1)
     Num=::NewEnumerator1
    )
    Spawn Group:NiagaraInt32=&MakeIntFromFloat(Float=1, Conversion Mode=::NewEnumerator1)
    Trigger:NiagaraBool=&Boolean_AndOperation(Bool1=$Emitter.UpdateValueStateChange.Changed, Bool2=&SetBoolByFloatComparison(A=$System.Age, B=0))
   ) {module=TriggerSpawn001}
   &self:TriggerSpawn(
    Spawn Count:NiagaraInt32=&MergeHelper_IntSwitch(
     0=&RandomRangeInt(Minimum=4, Maximum=8)
     1=&RandomRangeInt(Minimum=4, Maximum=8)
     2=&RandomRangeInt(Minimum=4, Maximum=8)
     3=?
     4=?
     5=?
     Index=&MakeIntFromFloat(Float=$User.BurstIndex, Conversion Mode=::NewEnumerator1)
     Num=::NewEnumerator1
    )
    Spawn Group:NiagaraInt32=&MakeIntFromFloat(Float=2, Conversion Mode=::NewEnumerator1)
    Trigger:NiagaraBool=&Boolean_AndOperation(Bool1=$Emitter.UpdateValueStateChange.Changed, Bool2=&SetBoolByFloatComparison(A=$System.Age, B=0))
   ) {module=TriggerSpawn003}
   &self:TriggerSpawn(
    Spawn Count:NiagaraInt32=&MergeHelper_IntSwitch(
     0=1
     1=1
     2=1
     3=?
     4=?
     5=?
     Index=&MakeIntFromFloat(Float=$User.BurstIndex, Conversion Mode=::NewEnumerator1)
     Num=::NewEnumerator1
    )
    Spawn Group:NiagaraInt32=&MakeIntFromFloat(Float=3, Conversion Mode=::NewEnumerator1)
    Trigger:NiagaraBool=&Boolean_AndOperation(Bool1=$Emitter.UpdateValueStateChange.Changed, Bool2=&SetBoolByFloatComparison(A=$System.Age, B=0))
   ) {module=TriggerSpawn002}
  ]
  particleSpawnScript [
   /SetVariables(Particles.VisibilityTag=&SpawnGroup()) {isSetParameters=true}
   &InitializeParticle(
    Color=&MergeHelper_ColorSwitch(
     0=<rgba 1 0.7271 0 1>
     1=<rgba 1 1 1 1>
     2=<rgba 1 0.7271 0 1>
     3=<rgba 3 1.2863 0 0.287>
     Index=&SpawnGroup()
     Num=::NewEnumerator2
    )
    Lifetime=&MergeHelper_FloatSwitch(
     0=&RandomRangeFloat(Minimum=0.1, Maximum=0.2)
     1=0.35
     2=&RandomRangeFloat(Minimum=0.2, Maximum=0.3)
     3=0.106
     Index=&SpawnGroup()
     Num=::NewEnumerator2
    )
    Mass Max=1.5
    Mass Mode=::NewEnumerator3
    Position Offset=&MergeHelper_VectorSwitch(
     0=<-74 52 103>
     1=<25 47 116>
     2=<-33 19 170>
     3=?
     Index=&MakeIntFromFloat(Float=$User.BurstIndex)
     Num=::NewEnumerator1
    )
    Sprite Rotation Mode=::NewEnumerator2
    Sprite Size Mode=::NewEnumerator3
    Uniform Sprite Size=&MergeHelper_FloatSwitch(
     0=63.0079
     1=67.3367
     2=&RandomRangeFloat(Minimum=2, Maximum=5)
     3=57.4304
     Index=&SpawnGroup()
     Num=::NewEnumerator2
    )
    UsePositionOffset=true
    Write Material Random=true
   )
   &AddVelocity(
    Module.Use Falloff Distance=false
    Origin Offset=&MergeHelper_VectorSwitch(
     0=<-74 52 103>
     1=<25 47 116>
     2=<-33 19 170>
     3=?
     Index=&MakeIntFromFloat(Float=$User.BurstIndex)
     Num=::NewEnumerator1
    )
    Override Seed=false
    Velocity Mode=::NewEnumerator1
    Velocity Speed=&MergeHelper_FloatSwitch(0=0, 1=0, 2=199.8638, 3=?, Index=&SpawnGroup(), Num=::NewEnumerator1)
   )
   &CameraOffset(Camera Offset Amount=15)
  ]
  particleUpdateScript [
   &DynamicMaterialParameters(
    Index 0 Param 1=&FloatFromCurve(FloatCurve=DataInterface() {propertyValues {Curve {defaultValue=1, keys [{tangentMode=RCTM_User, time=0, value=0}, {tangentMode=RCTM_User, time=1, value=1}]}}})
    Index 0 Param 2=1
    Module.Index0Float0Write=true
    Module.Index0Float1Write=true
    Module.Index0Float2Write=true
    Module.Index0Float3Write=true
   )
   &ParticleState()
   &SpriteRotationRate(Rotation Rate=&MergeHelper_FloatSwitch(
     0=0
     1=0
     2=&SelectFloatFromArray(
      Array Sampling Mode=::NewEnumerator0
      Direct Array Index=?
      Float Selection Array=DataInterface() {propertyValues {FloatData [-360,360,180,-180]}}
      Random Seed=0
     )
     3=74.4845
     Index=&SpawnGroup()
     Num=::NewEnumerator2
    ))
   &ScaleSpriteSize(
    Scale Sprite Size Mode=::NewEnumerator0
    Uniform Scale Factor=&MergeHelper_FloatSwitch(
     0=&FloatFromCurve(FloatCurve=DataInterface() {propertyValues {Curve {keys [
          {arriveTangent=0.5905, interpMode=RCIM_Cubic, leaveTangent=0.5905, tangentMode=RCTM_User, time=0, value=0}
          {arriveTangent=-0.0124, interpMode=RCIM_Cubic, leaveTangent=-0.0124, tangentMode=RCTM_User, time=0.4057, value=0.9642}
          {arriveTangent=-1.0652, interpMode=RCIM_Cubic, leaveTangent=-1.0652, tangentMode=RCTM_User, time=1, value=0}
         ]}}})
     1=&FloatFromCurve(FloatCurve=DataInterface() {propertyValues {Curve {keys [
          {arriveTangent=14.5835, interpMode=RCIM_Cubic, leaveTangent=14.5835, tangentMode=RCTM_User, time=0, value=0}
          {arriveTangent=0.1815, interpMode=RCIM_Cubic, leaveTangent=0.1815, tangentMode=RCTM_User, time=0.1, value=1}
          {arriveTangent=-0.0124, interpMode=RCIM_Cubic, leaveTangent=-0.0124, tangentMode=RCTM_User, time=0.6018, value=1}
         ]}}})
     2=&FloatFromCurve(FloatCurve=DataInterface() {propertyValues {Curve {keys [<at 0.82 1>,<at 1 0>]}}})
     3=&FloatFromCurve(FloatCurve=DataInterface() {propertyValues {Curve {defaultValue=0, keys [{arriveTangent=10.9439, tangentMode=RCTM_User, time=0.2695, value=1}]}}})
     Index=&SpawnGroup()
     Num=::NewEnumerator2
    )
   )
   &GravityForce(Gravity=&MergeHelper_VectorSwitch(0=<0 0 0>, 1=<0 0 0>, 2=<0 0 -574.9391>, 3=?, Index=&SpawnGroup(), Num=::NewEnumerator1))
   &SolveForcesAndVelocity(Clamp Velocity=false, Limit Acceleration=false)
   &ScaleColor(
    Scale Alpha=&MergeHelper_FloatSwitch(
     0=1
     1=1
     2=1
     3=&FloatFromCurve(FloatCurve=DataInterface() {propertyValues {Curve {
         defaultValue=1
         keys [{tangentMode=RCTM_User, time=0, value=-0.4}, {tangentMode=RCTM_User, time=0.22, value=1}, <at 0.5809 1.0194>, <at 0.9687 0>]
        }}})
     4=0
     Index=&SpawnGroup()
     Num=::NewEnumerator3
    )
    ScaleA=true
    ScaleRGB=true
   )
  ]
  rendererClasses [&NiagaraSpriteRendererProperties]
  renderers [
   {rendererClass=&NiagaraSpriteRendererProperties}
   {rendererClass=&NiagaraSpriteRendererProperties, rendererIndex=1}
   {rendererClass=&NiagaraSpriteRendererProperties, rendererIndex=2}
   {rendererClass=&NiagaraSpriteRendererProperties, rendererIndex=3}
  ]
 }
 Sparkles {
  emitterUpdateScript [
   &EmitterState(Life Cycle Mode=::NewEnumerator1, Loop Behavior=::NewEnumerator1, Loop Delay=12, Loop Duration=1.5, UseLoopDelay=true)
   &SpawnBurst_Instantaneous(Spawn Group=1, Use Spawn Probability=false, UseLoopCountLimit=false)
   &SpawnRate(SpawnRate=6, Use Spawn Probability=false)
  ]
  particleSpawnScript [
   /SetVariables(Particles.VisibilityTag=&SpawnGroup()) {isSetParameters=true}
   &InitializeParticle(
    Color=<rgba 3.6655 4.0104 3.5532 1>
    Lifetime=&MergeHelper_FloatSwitch(
     0=&RandomRangeFloat(Minimum=0.3, Maximum=0.8)
     1=1
     2=?
     3=?
     Index=&SpawnGroup()
     Num=::NewEnumerator0
    )
    Mass Max=1.5
    Mass Mode=::NewEnumerator3
    Ribbon Width=30
    Ribbon Width Mode=::NewEnumerator1
    Sprite Rotation Mode=::NewEnumerator2
    Sprite Size Mode=::NewEnumerator3
    Uniform Sprite Size=&MergeHelper_FloatSwitch(
     0=&RandomRangeFloat(Minimum=5, Maximum=10)
     1=30
     2=?
     3=?
     Index=&SpawnGroup()
     Num=::NewEnumerator0
    )
    UsePositionOffset=false
    Write Material Random=true
   )
   &CameraOffset(Camera Offset Amount=12)
  ]
  particleUpdateScript [
   &SkeletalMeshLocation(
    Bone / Socket Index=&SelectIntFromArray(Direct Array Index=&SpawnGroup(), Int Selection Array=DataInterface() {propertyValues {IntData [0,1]}})
    MaskSpawnGroup=false
    Mesh Sampling Type=::NewEnumerator5
    Orientation Sampling=::NewEnumerator1
    Skeletal Mesh=DataInterface() {propertyValues {
      DefaultMesh=null
      ExcludeBoneName=root
      OptionalComponentClass=null
      PreviewMesh=&ParadeLotus_Gadget
      SkinningMode=SkinOnTheFly
      SoftSourceActor=null
      SourceMode=Default
      UvSetIndex=0
      WholeMeshLOD=-1
      bExcludeBone=false
      bReadDeformedGeometry=true
      bRequireCurrentFrameData=false
      ComponentTags []
      FilteredBones []
      FilteredSockets [Sword01_LOD0_JNTSocket, TipSword01_LOD0_JNTSocket]
      MeshUserParameter {parameter {name=None, typeDefHandle {registeredTypeIndex=1}, varData []}}
      SamplingRegions []
     }}
    Socket Sampling Mode=::NewEnumerator5
    Write To Particle Position=false
   )
   &ShapeLocation(
    Additional Quaternion Rotation=false
    Box Size=<5 5 53.8764>
    MaskSpawnGroup=true
    Offset Mode=::NewEnumerator0
    Override Seed=false
    Rotation Mode=::NewEnumerator3
    Rotation Quaternion=$Particles.SkeletalMeshLocation.SampledRotation
    Shape Origin=$Particles.SkeletalMeshLocation.SampledPosition
    Shape Primitive=::NewEnumerator2
    Surface Only=false
    UseOwnerScale=true
   )
   &ParticleState()
   &SpriteRotationRate(Rotation Rate=&MakeCustomFloatFromBool(Bool=&RandomBool(), True Float=180, False Float=-180))
   &ScaleSpriteSize(Uniform Curve Sprite Scale=DataInterface() {propertyValues {Curve {
       defaultValue=0
       keys [
        {interpMode=RCIM_Cubic, tangentMode=RCTM_User, time=0, value=0}
        {interpMode=RCIM_Cubic, tangentMode=RCTM_User, time=0.5, value=1}
        {tangentMode=RCTM_User, time=1, value=0}
       ]
      }}})
   &SolveForcesAndVelocity(Clamp Velocity=false, Limit Acceleration=false)
  ]
  rendererClasses [&NiagaraSpriteRendererProperties]
  renderers [
   {rendererClass=&NiagaraSpriteRendererProperties}
   {rendererClass=&NiagaraSpriteRendererProperties, rendererIndex=1}
  ]
 }
 SwordTrail {
  emitterUpdateScript [
   &EmitterState(
    Inactive Response=::NewEnumerator1
    Life Cycle Mode=::NewEnumerator1
    Loop Behavior=::NewEnumerator1
    Loop Duration Mode=::NewEnumerator1
    UseLoopDelay=false
   )
   &SpawnRate(
    SpawnRate=&Abs_Float(Float:NiagaraFloat=&Multiply_Float(A=$User.RibbonCurve, B=400))
    Use Spawn Probability=false
   )
  ]
  particleSpawnScript [
   &InitializeParticle(
    Lifetime=&MakeCustomFloatFromBool(
     Bool=&SetBoolByFloatComparison(A=$User.RibbonCurve, B=0.1, Comparison Type=::NewEnumerator6)
     True Float=0.066
     False Float=0.03
    )
    Mass Max=1.5
    Mass Mode=::NewEnumerator3
    Ribbon Width=30
    Ribbon Width Mode=::NewEnumerator1
    Sprite Rotation Mode=::NewEnumerator2
    Sprite Size Mode=::NewEnumerator1
    Uniform Sprite Size Max=8
    Uniform Sprite Size Min=4
    UsePositionOffset=false
    Write Material Random=true
   )
   &SkeletalMeshLocation(
    MaskSpawnGroup=false
    Mesh Sampling Type=::NewEnumerator5
    Orientation Sampling=::NewEnumerator1
    Sampled Position Offset=&MakeCustomVectorFromBool(Bool=&SetBoolByFloatComparison(A=$User.RibbonCurve, B=0), True Vector=<0 0 1>, False Vector=<0 -10 23.41>)
    Skeletal Mesh=DataInterface() {propertyValues {
      DefaultMesh=null
      ExcludeBoneName=root
      OptionalComponentClass=null
      PreviewMesh=&ParadeLotus_Gadget
      SkinningMode=SkinOnTheFly
      SoftSourceActor=null
      SourceMode=Default
      UvSetIndex=0
      WholeMeshLOD=-1
      bExcludeBone=false
      bReadDeformedGeometry=true
      bRequireCurrentFrameData=false
      ComponentTags []
      FilteredBones []
      FilteredSockets [Sword02_LOD0_JNTSocket]
      MeshUserParameter {parameter {name=None, typeDefHandle {registeredTypeIndex=1}, varData []}}
      SamplingRegions []
     }}
    Socket Sampling Mode=::NewEnumerator5
   )
  ]
  particleUpdateScript [
   &ParticleState()
   /SetVariables(Particles.RibbonFace:Vector3f=&RotateVectorByQuaternion(Vector To Rotate=<0 0 1>, Quaternion=$Particles.SkeletalMeshLocation.SampledRotation)) {isSetParameters=true}
   /SetVariables(
    Particles.Ribbon1Color:LinearColor=&Multiply_LinearColorByFloat(
     Linear Color=&FNTimeOfDayColor(
      Night Color=<rgba 0.3472 0.3959 0.75 1>
      Morning Color=<rgba 0.6823 0.6321 0.5554 1>
      Day Color=<rgba 1 1 1 1>
      Evening Color=<rgba 0.6823 0.5467 0.462 1>
     )
     Float=&MakeCustomFloatFromBool(Bool=&SetBoolByFloatComparison(A=$User.RibbonCurve, B=0), True Float=1, False Float=0)
    )
    Particles.Ribbon2Color:LinearColor=&Multiply_LinearColorByFloat(
     Linear Color=&FNTimeOfDayColor(
      Night Color=<rgba 0.3472 0.3959 0.75 1>
      Morning Color=<rgba 0.6823 0.6321 0.5554 1>
      Day Color=<rgba 1 1 1 1>
      Evening Color=<rgba 0.6823 0.5467 0.462 1>
     )
     Float=&MakeCustomFloatFromBool(
      Bool=&SetBoolByFloatComparison(A=$User.RibbonCurve, B=-0.01, Comparison Type=::NewEnumerator6)
      True Float=1
      False Float=0
     )
    )
    Particles.Ribbon2Wiedth=60
   ) {isSetParameters=true}
   &ScaleRibbonWidth(Ribbon Width Scale=$User.RibbonCurve)
   &InheritVelocity(Speed Limit=false, UseSpeedThreshold=false)
   &SolveForcesAndVelocity(Clamp Velocity=false, Limit Acceleration=false)
  ]
  rendererClasses [&NiagaraRibbonRendererProperties]
  renderers [
   {rendererClass=&NiagaraRibbonRendererProperties}
   {rendererClass=&NiagaraRibbonRendererProperties, rendererIndex=1}
  ]
 }
}
summary {
 systemName=NS_Emote_ParadeLotus_Ribbons
 emitters [
  {emitterName=SwordTrail, rendererClasses [&NiagaraRibbonRendererProperties]}
  {emitterName=Punchs, rendererClasses [&NiagaraSpriteRendererProperties]}
  {emitterName=BurstLines, rendererClasses []}
  {emitterName=Sparkles, rendererClasses [&NiagaraSpriteRendererProperties]}
  {emitterName=Arrows, rendererClasses [&NiagaraMeshRendererProperties]}
  {emitterName=ArrowSmearLines, rendererClasses []}
 ]
 userVariables [
  User.BurstIndex {type=NiagaraFloat, defaultValue=0, description=""}
  User.RibbonCurve {type=NiagaraFloat, defaultValue=-1, description=""}
 ]
}
